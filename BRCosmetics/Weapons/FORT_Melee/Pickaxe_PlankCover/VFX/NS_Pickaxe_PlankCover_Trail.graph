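# Niagara ribbon trail effect for the PlankCover pickaxe, made of short-lived Spoiler and Wheel anim trails.
# Spoiler spawns three trails whose colors, widths and some socket pairs come from arrays indexed by the User.Variant float as an integer, covering eleven variants.
# Wheel adds one dark, camera-offset trail with fixed width and color.
# Both fade quickly on system age and scale color by a time-of-day multiplier.
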NiagaraSystem /BRCosmetics/Weapons/FORT_Melee/Pickaxe_PlankCover/VFX/NS_Pickaxe_PlankCover_Trail
dependencies {
 usedDataInterfaces [
  &NiagaraDataInterfaceArrayInt32
  &NiagaraDataInterfaceArrayColor
  &NiagaraDataInterfaceArrayFloat
  &NiagaraDataInterfaceColorCurve
  &NiagaraDataInterfaceSocketReader
 ]
 usedDynamicInputs [&MakeIntFromFloat, &ColorFromCurve]
 usedModules [
  &self:SystemSpawnScript.SystemScriptSource.SystemScriptGraph.NiagaraNodeAssignment_2.SetVariables_#1
  &self:SystemSpawnScript.SystemScriptSource.SystemScriptGraph.NiagaraNodeAssignment_1.SetVariables_#2
  &self:SystemSpawnScript.SystemScriptSource.SystemScriptGraph.NiagaraNodeAssignment_3.SetVariables_#3
  &SystemState
 ]
}
emitters {
 Spoiler {
  emitterUpdateScript [
   &EmitterState(Life Cycle Mode=::NewEnumerator1, Loop Behavior=::NewEnumerator1, Loop Duration=0.25, UseLoopDelay=true)
   &FN_SpawnAnimTrail(Number of Anim Trails=3)
  ]
  particleSpawnScript [
   &FN_InitializeAnimTrailBoneIndecies(First Bone or Socket Index=2, Second Bone or Socket Index=3)
   &FN_InitializeAnimTrailBoneIndecies(
    AnimTrail Selection=1
    First Bone or Socket Index:NiagaraInt32=&SelectIntFromArray(Direct Array Index=$System.VariantFromFloat, Int Selection Array=DataInterface() {propertyValues {IntData [3,3,3,3,3,3,3,3,3,3,8]}})
    Second Bone or Socket Index:NiagaraInt32=&SelectIntFromArray(Direct Array Index=$System.VariantFromFloat, Int Selection Array=DataInterface() {propertyValues {IntData [5,5,5,7,7,5,5,5,5,7,7]}})
   ) {module=FN_InitializeAnimTrailBoneIndecies001}
   &FN_InitializeAnimTrailBoneIndecies(
    AnimTrail Selection=2
    First Bone or Socket Index:NiagaraInt32=&SelectIntFromArray(Direct Array Index=$System.VariantFromFloat, Int Selection Array=DataInterface() {propertyValues {IntData [2,2,2,2,2,2,2,2,2,2,8]}})
    Second Bone or Socket Index:NiagaraInt32=&SelectIntFromArray(Direct Array Index=$System.VariantFromFloat, Int Selection Array=DataInterface() {propertyValues {IntData [4,4,4,6,6,4,4,4,4,6,6]}})
   ) {module=FN_InitializeAnimTrailBoneIndecies002}
   &FN_InitializeAnimTrail(
    AnimTrailWidth=&MergeHelper_FloatSwitch(
     0=1
     1=&SelectFloatFromArray(Direct Array Index=$System.VariantFromFloat, Float Selection Array=DataInterface() {propertyValues {FloatData [1,1,1,1,1,1,1,1,1,1,0.7]}})
     2=&SelectFloatFromArray(Direct Array Index=$System.VariantFromFloat, Float Selection Array=DataInterface() {propertyValues {FloatData [1,1,1,1,1,1,1,1,1,1,1]}})
     3=?
     Index=&SpawnGroup()
     Num=::NewEnumerator1
    )
    Color=&MergeHelper_ColorSwitch(
     0=&SelectLinearColorFromArray(
      Array Sampling Mode=::NewEnumerator1
      Direct Array Index=$System.VariantFromFloat
      Color Selection Array=DataInterface() {propertyValues {ColorData [
         <rgba 0.186 0.07 0.35 0.1>
         <rgba 0.4591 0.0712 0.0777 0.1>
         <rgba 0.018 0.2061 0.0257 0.1>
         <rgba 0.0155 0.1125 0.5 0.075>
         <rgba 0.4978 0.0641 0.4882 0.075>
         <rgba 0.0155 0.1125 0.5 0.075>
         <rgba 0.6585 0.238 0.0009 0.075>
         <rgba 0.275 0.049 0.0413 0.1>
         <rgba 0.0091 0.2061 0.1504 0.1>
         <rgba 0.0495 0.0443 0.075 0.2>
         <rgba 0.06 0.0666 0.15 0.1>
        ]}}
      Random Seed=?
     )
     1=&SelectLinearColorFromArray(
      Array Sampling Mode=::NewEnumerator1
      Direct Array Index=$System.VariantFromFloat
      Color Selection Array=DataInterface() {propertyValues {ColorData [
         <rgba 0.186 0.07 0.35 0.1>
         <rgba 0.019 0.0195 0.025 0.1>
         <rgba 0.018 0.2061 0.0257 0.1>
         <rgba 0.0203 0.0513 0.15 0.125>
         <rgba 0.0262 0.3956 0.6585 0.075>
         <rgba 0.0203 0.0513 0.15 0.125>
         <rgba 0.4442 0.1125 0.0375 0.075>
         <rgba 0.25 0.025 0.0288 0.15>
         <rgba 0.0079 0.09 0.0484 0.1>
         <rgba 0.1035 0.4478 0.5 0.075>
         <rgba 0.35 0.2393 0.0179 0.1>
        ]}}
      Random Seed=?
     )
     2=&SelectLinearColorFromArray(
      Array Sampling Mode=::NewEnumerator1
      Direct Array Index=$System.VariantFromFloat
      Color Selection Array=DataInterface() {propertyValues {ColorData [
         <rgba 0.186 0.07 0.35 0.1>
         <rgba 0.019 0.0195 0.025 0.1>
         <rgba 0.019 0.0195 0.025 0.1>
         <rgba 0.0203 0.0513 0.15 0.125>
         <rgba 0.0262 0.3956 0.6585 0.075>
         <rgba 0.0203 0.0513 0.15 0.125>
         <rgba 0.4442 0.1125 0.0375 0.075>
         <rgba 0.25 0.025 0.0288 0.15>
         <rgba 0.0079 0.09 0.0484 0.1>
         <rgba 0.1035 0.4478 0.5 0.075>
         <rgba 0.25 0.025 0.0288 0.15>
        ]}}
      Random Seed=?
     )
     3=?
     Index=&SpawnGroup()
     Num=::NewEnumerator1
    )
    Lifetime=0.3
   )
  ]
  particleUpdateScript [
   &FN_UpdateAnimTrail()
   &ScaleColor(
    Curve Index=$Engine.System.Age
    Linear Color Curve=DataInterface() {propertyValues {
      AlphaCurve {keys [
        {arriveTangent=-26.5484, interpMode=RCIM_Cubic, leaveTangent=-26.5484, tangentMode=RCTM_User, time=0.1062, value=1}
        <at 0.1607 0 RCIM_Cubic>
       ]}
      BlueCurve {keys [<at 0 1>,<at 1 0.0662>]}
      GreenCurve {keys [<at 0 1>,<at 1 0.0662>]}
      RedCurve {keys [<at 0 1>,<at 1 0.0662>]}
     }}
    Scale Mode=::NewEnumerator2
   ) {module=ScaleColor001}
   &ScaleColor(Color Scale:LinearColor=$System.TODColorMultiplier, Scale Mode=3)
  ]
  rendererClasses [&NiagaraRibbonRendererProperties]
  renderers [{rendererClass=&NiagaraRibbonRendererProperties}]
 }
 Wheel {
  emitterUpdateScript [
   &EmitterState(Life Cycle Mode=::NewEnumerator1, Loop Behavior=::NewEnumerator1, Loop Duration=0.25, UseLoopDelay=true)
   &FN_SpawnAnimTrail()
  ]
  particleSpawnScript [
   &FN_InitializeAnimTrailBoneIndecies()
   &FN_InitializeAnimTrail(AnimTrailWidth=0.8, Color=<rgba 0.045 0.0405 0.05 0.75>, Lifetime=0.25)
   &CameraOffset(Camera Offset Amount=5)
  ]
  particleUpdateScript [
   &FN_UpdateAnimTrail()
   &ScaleColor(
    Curve Index=$Engine.System.Age
    Linear Color Curve=DataInterface() {propertyValues {
      AlphaCurve {keys [{arriveTangent=-36.5583, interpMode=RCIM_Cubic, leaveTangent=-36.5583, tangentMode=RCTM_User, time=0.075, value=1}, <at 0.15 0 RCIM_Cubic>]}
      BlueCurve {keys [<at 0 1>,<at 1 0.0662>]}
      GreenCurve {keys [<at 0 1>,<at 1 0.0662>]}
      RedCurve {keys [<at 0 1>,<at 1 0.0662>]}
     }}
    Scale Mode=::NewEnumerator2
   ) {module=ScaleColor001}
   &ScaleColor(Color Scale:LinearColor=$System.TODColorMultiplier, Scale Mode=3)
  ]
  rendererClasses [&NiagaraRibbonRendererProperties]
  renderers [{rendererClass=&NiagaraRibbonRendererProperties}]
 }
}
summary {
 systemName=NS_Pickaxe_PlankCover_Trail
 emitters [
  {emitterName=Spoiler, rendererClasses [&NiagaraRibbonRendererProperties]}
  {emitterName=Wheel, rendererClasses [&NiagaraRibbonRendererProperties]}
 ]
 userVariables [
  User.RibbonIDOffset {type=NiagaraInt32}
  User.Variant {type=NiagaraFloat, defaultValue=0, description=""}
 ]
}
